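# Post-process-style animation blueprint for the Buff Cat (Meowscles) body, layering secondary motion onto its parent mesh's copied pose.
# An emote-synced sequence drives the pec bones; weighted constraints and alpha-driven corrections adjust holsters and thighs.
# A sitting flag rotates the tail socket before a Control Rig pass.
# BuffcatSpyroomIdle part-start notifies set that flag; part-stop notifies clear it only for other contexts.
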AnimBlueprint /BRCosmetics/Characters/Player/Male/Large/Bodies/M_Large_Buff_Cat/Meshes/M_LRG_Buff_Cat_AnimBp
target_skeleton=&/M_LRG_Buff_Cat_Skeleton
graphs [
 AnimGraph {nodes [
   n1 = "Sequence Evaluator (by time)"!(
    Sequence:Anim Sequence Base Object Reference=Get SyncedSequence() -> (SyncedSequence:Anim Sequence Object Reference)
    ExplicitTime:float=Get MeowsclesMontagePosition() -> (MeowsclesMontagePosition:float)
   ) -> (Pose:Pose Link Structure) {is=title AnimGraphNode_SequenceEvaluator}
   n2 = Get IsPlayingEmoteAlpha() -> (IsPlayingEmoteAlpha:float)
   n3 = Copy Pose From Mesh() {bUseAttachedParent=true}
   n4 = "Constraint - dyn_holster_r_base (PPPPPPP)"!(
    ComponentPose=Local To Component(LocalPose=Layered blend per bone(BasePose=@n3, BlendPoses_0=@n1, BlendWeights_0=@n2) {layerSetup [{branchFilters [{boneName=dyn_peck_l}, {boneName=dyn_peck_r}]}]})
    Alpha=1
    ConstraintWeights_0=0
    ConstraintWeights_1:float=0.0712
    ConstraintWeights_2:float=0.0346
    ConstraintWeights_3:float=0.0259
    ConstraintWeights_4:float=0
    ConstraintWeights_5:float=0.82
    ConstraintWeights_6:float=0.0483
   ) -> (Pose) {
    alphaBoolBlend {blendOption=Linear}
    boneToModify {boneName=dyn_holster_r_base}
    constraintSetup [
     {targetBone {boneName=thigh_twist_01_r}}
     {targetBone {boneName=spine_01}}
     {targetBone {boneName=spine_02}}
     {targetBone {boneName=spine_03}}
     {targetBone {boneName=spine_04}}
     {targetBone {boneName=pelvis}}
     {targetBone {boneName=thigh_twist_01_l}}
    ]
    constraintWeights [1,1,1,1,1,1,1]
   }
   n5 = "Transform (Modify) Bone - Bone: dyn_holster_r_base"(
    ComponentPose=@n4
    Alpha=Get ThighRotationRightAlpha() -> (ThighRotationRightAlpha:float)
    Translation="0,0,0"
    Rotation="-10,0,0"
    Scale="1,1,1"
   ) {
    rotationMode=BMM_Additive
    rotationSpace=BCS_BoneSpace
    alphaBoolBlend {blendOption=Linear}
    boneToModify {boneName=dyn_holster_r_base}
    rotation {pitch=-10}
   }
   n6 = "Constraint - dyn_holster_l_base (PPPPP)"!(
    ComponentPose=@n5
    Alpha=1
    ConstraintWeights_0=0.8128
    ConstraintWeights_1:float=0.037
    ConstraintWeights_2:float=0.0304
    ConstraintWeights_3:float=0.0464
    ConstraintWeights_4:float=0.0734
   ) -> (Pose) {
    alphaBoolBlend {blendOption=Linear}
    boneToModify {boneName=dyn_holster_l_base}
    constraintSetup [
     {targetBone {boneName=pelvis}}
     {targetBone {boneName=spine_02}}
     {targetBone {boneName=spine_03}}
     {targetBone {boneName=thigh_twist_01_l}}
     {targetBone {boneName=spine_01}}
    ]
    constraintWeights [1,1,1,1,1]
   }
   n7 = "Transform (Modify) Bone - Bone: dyn_holster_l_base"(
    ComponentPose=@n6
    Alpha=Get ThighRotationLeftAlpha() -> (ThighRotationLeftAlpha:float)
    Translation="0,0,0"
    Rotation="-10,0,0"
    Scale="1,1,1"
   ) {
    rotationMode=BMM_Additive
    rotationSpace=BCS_BoneSpace
    alphaBoolBlend {blendOption=Linear}
    boneToModify {boneName=dyn_holster_l_base}
    rotation {pitch=-10}
   }
   n8 = "Transform (Modify) Bone - Bone: sub_thigh_r"(ComponentPose=@n7, Alpha=Get ThighRotationRightAlpha(), Translation="0,0,0", Rotation="0,0,0", Scale="3,3,3") {scaleMode=BMM_Additive, scaleSpace=BCS_BoneSpace, alphaBoolBlend {blendOption=Linear}, boneToModify {boneName=sub_thigh_r}, scale=<3 3 3>}
   n9 = "Transform (Modify) Bone - Bone: sub_thigh_l"(ComponentPose=@n8, Alpha=Get ThighRotationLeftAlpha(), Translation="0,0,0", Rotation="0,0,0", Scale="20,6,4") {scaleMode=BMM_Additive, scaleSpace=BCS_BoneSpace, alphaBoolBlend {blendOption=Linear}, boneToModify {boneName=sub_thigh_l}, scale=<20 6 4>}
   n10 = "Transform (Modify) Bone - Bone: dyn_tail_socket"!(ComponentPose=@n9, bAlphaBoolEnabled:bool=Get IsSitting() -> (IsSitting:bool), Translation="0,0,0", Rotation="-75,0,0", Scale="1,1,1") -> (Pose) {
    alphaInputType=Bool
    rotationMode=BMM_Additive
    rotationSpace=BCS_BoneSpace
    translationMode=BMM_Additive
    translationSpace=BCS_BoneSpace
    alphaBoolBlend {blendOption=Linear}
    boneToModify {boneName=dyn_tail_socket}
    rotation {pitch=-75}
   }
   n11 = Control Rig(Source=Component To Local(ComponentPose=@n10)) {
    alphaBoolBlend {blendOption=Linear}
    controlRigAssetReference {blueprintRigClass=&/M_LRG_Buff_Cat_CrtlRig.M_LRG_Buff_Cat_CrtlRig_C}
   }
   Output Pose(Result=@n11) {layerGroup=None}
  ]}
 EventGraph {nodes [
   n1 = Event OnNotifyPartsStarted!() -> (OutputDelegate, then, Context:Name=None) {is=title K2Node_Event}
   > Branch(Condition="Equal (Name)"!(A:Name=@n1.Context, B:Name=BuffcatSpyroomIdle) -> (ReturnValue:bool) {is=title K2Node_CallFunction})
   > Set IsSitting(IsSitting:bool=true) -> (Output_Get:bool)
   n2 = Event OnNotifyPartsStopped!() -> (OutputDelegate, then, Context:Name=None) {is=title K2Node_Event}
   > n3 = Branch(Condition="Equal (Name)"(A=@n2.Context, B=BuffcatSpyroomIdle))
   Set IsSitting(execute=@n3.else, IsSitting=false)
  ]}
]
variables [IsSitting {type=boolean}]
