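# Animation Blueprint for the large male Buff Cat face accessory mesh, driving its facial pose and ear jiggle.
# It copies pose and curves from the attached parent mesh, then layers the BuffCat facial sequence per bone, weighted by the DisableFaceOverride curve.
# Chained Anim Dynamics nodes then simulate the left and right ear bones with separately tuned limits, gravity and motion scaling.
# The event graph is empty.
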AnimBlueprint /BRCosmetics/Characters/Player/Male/Large/Bodies/M_Large_Buff_Cat/Meshes/Parts/M_LRG_Buff_Cat_FaceAcc_AnimBp
target_skeleton=/Game/Characters/Player/Male/Large/Bodies/M_Large_Buff_Cat/Meshes/Parts/M_LRG_Buff_Cat_FaceAcc_Skeleton
graphs [
 AnimGraph {nodes [
   n1 = Copy Pose From Mesh() {bCopyCurves=true, bUseAttachedParent=true}
   n2 = Sequence Player 'BuffCat_Male_Facial'() {sequence=/BRCosmetics/Animation/Game/MainPlayer/Facial/BuffCat/BuffCat_Male_Facial}
   n3 = float * float(A=1, B=GetCurveValue(CurveName=DisableFaceOverride))
   n4 = Anim Dynamics - Bone: dyn_ear_l(
    ComponentPose=Local To Component(LocalPose=Layered blend per bone(BasePose=@n1, BlendPoses_0=@n2, BlendWeights_0=@n3))
    Alpha=1
    ExternalForce="0,0,0"
   ) {
    bGravityOverrideInSimSpace=true
    bUseGravityOverride=true
    simulationSpace=Actor
    alphaBoolBlend {blendOption=Linear}
    boundBone {boneName=dyn_ear_l}
    componentAppliedLinearAccClamp=<2000 2000 2000>
    componentLinearAccScale=<0.5 0.5 0.5>
    componentLinearVelScale=<0.5 0.5 0.5>
    gravityOverride=<1 1 0.85>
    physicsBodyDefinitions [{
      sphereCollisionRadius=0
      boundBone {boneName=dyn_ear_l}
      boxExtents=<6 6 2>
      constraintSetup {angularLimitsMax=<0 25 0>, angularLimitsMin=<0 1 0>, angularTarget=<0 0 0>}
      localJointOffset=<2 0 0>
     }]
    relativeSpaceBone {boneName=spine_05}
   }
   n5 = Anim Dynamics - Bone: dyn_ear_r(ComponentPose=@n4, Alpha=0.75, ExternalForce="0,0,0") {
    bGravityOverrideInSimSpace=true
    bUseGravityOverride=true
    alphaBoolBlend {blendOption=Linear}
    boundBone {boneName=dyn_ear_r}
    componentAppliedLinearAccClamp=<2000 2000 2000>
    componentLinearAccScale=<0.25 0.25 0.25>
    componentLinearVelScale=<0.25 0.25 0.25>
    gravityOverride=<1 1 0.75>
    physicsBodyDefinitions [{
      sphereCollisionRadius=0
      boundBone {boneName=dyn_ear_r}
      boxExtents=<6 6 2>
      constraintSetup {angularLimitsMax=<0 3 0>, angularLimitsMin=<0 -15 0>, angularTarget=<0 0 0>}
      localJointOffset=<2 0 0>
     }]
    relativeSpaceBone {boneName=pelvis}
   }
   Output Pose(Result=Component To Local(ComponentPose=@n5)) {layerGroup=None}
  ]}
 EventGraph {links [], nodes []}
]
variables []
